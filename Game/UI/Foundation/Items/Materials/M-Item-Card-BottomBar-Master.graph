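# Translucent UI master material for the item card bottom bar, colored by rarity parameters.
# An offset radial gradient on normalized UI UVs blends the Rarity 3 and Rarity 2 colors into the base.
# A rotated, time-scrolled Frac/Abs gradient adds an optional brightened sweep via bAnimateColor.
# A smoothstepped, angled UV mask overlays an optional Rarity 4 strip via bDarkStrip, with gAlpha as an overall alpha control.
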Material /Game/UI/Foundation/Items/Materials/M-Item-Card-BottomBar-Master
graph {
 properties {
  bUsedWithStaticMesh=false
  blendMode=BLEND_Translucent
  materialDomain=MD_UI
  pixelDepthOffsetMode=PDOM_Legacy
  usedShadingModels=Default Lit
 }
 nodes [
  n1 = VectorParameter() {group=Vector Parameter Values, parameterName=Rarity 2, defaultValue=<rgba 0.1195 0.521 0 1>}
  n2 = MaterialFunctionCall() {materialFunction=&GetUserInterfaceUV}
  n3 = Multiply(A=@n2.Normalized UV, B=Constant2Vector() {g=0.43, r=1})
  n4 = Constant2Vector() {g=0.4, r=-0.5}
  n5 = LinearInterpolate(
   A=VectorParameter() {group=Vector Parameter Values, parameterName=Rarity 3, defaultValue=<rgba 0 0.2159 0.0015 1>}.RGB
   B=@n1.RGB
   Alpha=MaterialFunctionCall(UVs=Add(A=@n3, B=@n4), Radius=Constant() {r=1}, Density=Constant() {r=1.5}) {materialFunction=&RadialGradientExponential}.RadialGradientExponential
  )
  n6 = Rotator(Coordinate=TextureCoordinate() {vTiling=0.25}, Time=ScalarParameter() {defaultValue=-8.5, parameterName=gRotation})
  n7 = Multiply(A=Time(), B=ScalarParameter() {defaultValue=0.25, parameterName=gTime})
  n8 = Add(A=Frac(None=Add(A=@n6, B=@n7))) {constB=-0.5}
  n9 = MaterialFunctionCall(
   Alpha=ComponentMask(None=Abs(None=@n8)) {r=true}
   Min=ScalarParameter() {parameterName=gGradientMin}
   Max=ScalarParameter() {defaultValue=1, parameterName=gGradientMax}
  ) {materialFunction=&SmoothStep}
  n10 = LinearInterpolate(
   A=@n5
   B=LinearInterpolate(A=@n5, B=Add(A=@n1.RGB) {constB=0.25}, Alpha=Clamp(None=@n9.Result))
   Alpha=ScalarParameter() {group=Scalar Parameter Values, parameterName=bAnimateColor, sliderMax=1}
  )
  n11 = Constant2Vector() {g=0.1}
  n12 = ComponentMask(None=Rotator(Coordinate=Add(A=@n2.Normalized UV, B=@n11), Time=ScalarParameter() {defaultValue=12.25, parameterName=gBarAngel})) {g=true}
  n13 = MaterialFunctionCall(Alpha=@n12, Min=Min(A=@n12) {constB=0.45}, Max=Max(A=@n12) {constB=0.475}) {materialFunction=&SmoothStep}
  LinearInterpolate(
   A=@n10
   B=LinearInterpolate(
    A=@n10
    B=VectorParameter() {group=Vector Parameter Values, parameterName=Rarity 4, defaultValue=<rgba 0.0006 0.0782 0.0009 1>}.RGB
    Alpha=Clamp(None=@n13.Result)
   )
   Alpha=ScalarParameter() {parameterName=bDarkStrip, sliderMax=1}
  )
  ScalarParameter() {defaultValue=1, parameterName=gAlpha, sliderMax=1}
 ]
}
