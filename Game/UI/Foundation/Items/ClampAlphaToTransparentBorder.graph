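# ClampAlphaToTransparentBorder is a UI material function that zeroes alpha outside the 0–1 UV range, giving sampling beyond the edges a transparent border.
# It takes UVs and a scalar Alpha into a Custom HLSL node that builds a box mask from the difference of two step functions per axis.
# The product of both axes multiplies Alpha and is returned as the single scalar output.
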MaterialFunction /Game/UI/Foundation/Items/ClampAlphaToTransparentBorder
graph {nodes [
  MaterialExpressionFunctionInput_1 = FunctionInput() {inputName=UVs, inputType=FunctionInput_Vector2}
  MaterialExpressionFunctionInput_0 = FunctionInput() {inputName=Alpha, inputType=FunctionInput_Scalar, sortPriority=1}
  n1 = Custom!(uv=@MaterialExpressionFunctionInput_1, color=@MaterialExpressionFunctionInput_0) {
   code="float2 box = step(float2(0, 0), uv) - step(float2(1,1), uv);\r\nreturn (box.x * box.y) * color;"
   outputType=CMOT_Float1
   inputs [
    {inputName=uv, input {expression=&self:MaterialExpressionFunctionInput_1}}
    {inputName=color, input {expression=&self:MaterialExpressionFunctionInput_0}}
   ]
  }
  FunctionOutput(None=@n1)
 ]}
