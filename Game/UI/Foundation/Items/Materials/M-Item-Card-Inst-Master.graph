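# Translucent UI master material for item cards, covering item, hero, lead and schematic looks tinted by four rarity colors.
# Scalar flags lerp between backgrounds: radial gradients, angled hero bands with panning triangle textures, and a procedural schematic grid.
# The Icon texture, framed per card type, composites over an offset shadow copy, with an animated tint for schematics.
# A distance-based border with gradient and optional animated sweep goes on top.
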Material /Game/UI/Foundation/Items/Materials/M-Item-Card-Inst-Master
graph {
 properties {
  bUsedWithStaticMesh=false
  blendMode=BLEND_Translucent
  materialDomain=MD_UI
  pixelDepthOffsetMode=PDOM_Legacy
  usedShadingModels=Default Lit
 }
 nodes [
  Constant2Vector() {g=-0.3683}
  Constant() {r=0.9}
  n1 = ScalarParameter() {group=General, parameterName=bIsItem, sliderMax=1}
  LinearInterpolate(Alpha=@n1) {constA=1.8, constB=1.8}
  n2 = MaterialFunctionCall() {materialFunction=&GetUserInterfaceUV}
  n3 = Multiply(
   A=@n2.Normalized UV
   B=AppendVector(
    A=ScalarParameter() {defaultValue=0.775, desc="Card Proportion (X +25%)", group=Portraits, parameterName=gIcon-Portrait-X Scale}
    B=ScalarParameter() {defaultValue=1.25, desc="Card Proportion (Y +25%)", group=Portraits, parameterName=gIcon-Portrait-Y Scale}
   )
  ) {constB=1.25}
  n4 = AppendVector(
   A=ScalarParameter() {defaultValue=0.1125, desc="X Offset for Card Proportion (X +25%)", group=Portraits, parameterName=gIcon-Portrait-X Offset}
   B=ScalarParameter() {defaultValue=-0.02, group=Portraits, parameterName=gIcon-Portrait-Y Offset}
  )
  n5 = Multiply(
   A=@n2.Normalized UV
   B=AppendVector(
    A=ScalarParameter() {defaultValue=1.0667, desc="Card Proportion (X=100%)", group=Schematic, parameterName=gIcon-Schematic-X Scale}
    B=ScalarParameter() {defaultValue=1.725, desc="Card Proportion (Y +27.5%)", group=Schematic, parameterName=gIcon-Schematic-Y Scale}
   )
  )
  n6 = AppendVector(
   A=ScalarParameter() {defaultValue=-0.0333, desc="X Offset for Card Proportion (X=100%)", group=Schematic, parameterName=gIcon-Schematic-X-Offset}
   B=ScalarParameter() {defaultValue=-0.0906, group=Schematic, parameterName=gIcon-Schematic-Y-Offset}
  )
  n7 = ScalarParameter() {group=General, parameterName=bIsSchematic, sliderMax=1}
  n8 = LinearInterpolate(
   A=ScalarParameter() {defaultValue=1.1, parameterName=gIconScaleNoBorder}
   B=ScalarParameter() {defaultValue=1.25, group=Items, parameterName=gIcon-Scale}
   Alpha=ScalarParameter() {defaultValue=1, group=Border, parameterName=bBorderOn, sliderMax=1}
  )
  n9 = Subtract(B=@n8)
  n10 = LinearInterpolate(
   A=LinearInterpolate(A=Add(A=@n3, B=@n4), B=Add(A=@n5, B=@n6), Alpha=@n7)
   B=Add(A=Multiply(A=@n2.Normalized UV, B=@n8) {constB=1.25}, B=Multiply(A=@n9) {constB=0.5})
   Alpha=ScalarParameter() {group=General, parameterName=bIsItem, sliderMax=1}
  )
  n11 = Add(A=@n10, B=Constant2Vector() {g=-0.025, r=-0.0125}) {constB=-0.05}
  n12 = TextureSampleParameter2D(UVs=@n11) {
   group=Icon
   parameterName=Icon
   samplerSource=SSM_Clamp_WorldGroupSettings
   texture=/Game/Widgets/SMyTownMapWidget/Common/BonusItemBackground
  }
  n13 = Multiply(A=MaterialFunctionCall!(UVs=@n11, Alpha=@n12.A) {materialFunction=/Game/UI/Foundation/Items/ClampAlphaToTransparentBorder}.Result) {constB=0.35}
  n14 = Constant3Vector() {constant=<rgba 0 0 0 1>}
  n15 = MaterialFunctionCall(
   UVs=@n2.Normalized UV
   CenterPosition=Constant2Vector() {g=0.5, r=0.5}
   Radius=LinearInterpolate(Alpha=ScalarParameter() {defaultValue=1, group=Border, parameterName=bBorderOn, sliderMax=1}) {constA=1.2, constB=0.9}
   Density=Constant() {r=1.8}
  ) {materialFunction=&RadialGradientExponential}
  n16 = Multiply(A=Multiply(
    A=@n15.RadialGradientExponential
    B=VectorParameter() {group=Rarity Colors, parameterName=Rarity 3, defaultValue=<rgba 0 0.2159 0.0015 1>}.RGB
   ) {constB=2}) {constB=0.75}
  n17 = MaterialFunctionCall() {materialFunction=&GetUserInterfaceUV}
  n18 = ComponentMask(None=Rotator(Coordinate=@n17.Normalized UV, Time=ScalarParameter() {defaultValue=4.25, group=Border, parameterName=gGradientAngle})) {r=true}
  n19 = Add(A=@n18) {constB=0}
  n20 = ScalarParameter() {group=Border, parameterName=gGradientMin}
  n21 = ScalarParameter() {defaultValue=1, group=Border, parameterName=gGradientMax}
  n22 = MaterialFunctionCall(Alpha=@n18, Min=@n20, Max=@n21) {materialFunction=&SmoothStep}
  n23 = VectorParameter() {group=Rarity Colors, parameterName=Rarity 1, defaultValue=<rgba 0.3564 0.9911 0 1>}
  n24 = ScalarParameter() {group=General, parameterName=bIsItem, sliderMax=1}
  n25 = LinearInterpolate(
   A=VectorParameter() {group=Rarity Colors, parameterName=Rarity 2, defaultValue=<rgba 0.1195 0.521 0 1>}.RGB
   B=Multiply(A=@n23.RGB) {constB=0.85}
   Alpha=@n24
  )
  n26 = LinearInterpolate(
   A=LinearInterpolate(
    A=VectorParameter() {group=Rarity Colors, parameterName=Rarity 4, defaultValue=<rgba 0.0006 0.0782 0.0009 1>}.RGB
    B=VectorParameter() {group=Rarity Colors, parameterName=Rarity 3, defaultValue=<rgba 0 0.2159 0.0015 1>}.RGB
    Alpha=@n24
   )
   B=@n25
   Alpha=Clamp(None=DotProduct(A=@n19, B=@n22.Result))
  )
  n27 = ScalarParameter() {defaultValue=-8.5, group=Border, parameterName=gRotation}
  n28 = Time()
  n29 = ScalarParameter() {defaultValue=0.25, group=General, parameterName=gTime}
  n30 = Frac(None=Add(A=Rotator(Coordinate=@n17.Normalized UV, Time=@n27), B=Multiply(A=@n28, B=@n29)))
  n31 = ComponentMask(None=Abs(None=Add(A=@n30) {constB=-0.5})) {r=true}
  n32 = MaterialFunctionCall(Alpha=@n31, Min=@n20, Max=@n21) {materialFunction=&SmoothStep}
  n33 = LinearInterpolate(
   A=@n26
   B=LinearInterpolate(A=@n26, B=Add(A=@n23.RGB) {constB=0.25}, Alpha=Clamp(None=@n32.Result))
   Alpha=ScalarParameter() {group=Border, parameterName=bAnimateBorder}
  )
  n34 = LinearInterpolate(
   A=VectorParameter() {group=Rarity Colors, parameterName=Rarity 4, defaultValue=<rgba 0.0006 0.0782 0.0009 1>}.RGB
   B=VectorParameter() {group=Rarity Colors, parameterName=Rarity 2, defaultValue=<rgba 0.1195 0.521 0 1>}.RGB
   Alpha=MaterialFunctionCall(
    UVs=@n2.Normalized UV
    CenterPosition=Constant2Vector() {g=0.5, r=0.5}
    Radius=ScalarParameter() {defaultValue=0.75, group=Items, parameterName=ItemBackRadius}
    Density=ScalarParameter() {defaultValue=1.3826, group=Items, parameterName=ItemBackDensity}
   ) {materialFunction=&RadialGradientExponential}.RadialGradientExponential
  )
  n35 = Desaturation(
   None=LinearInterpolate(A=LinearInterpolate(A=@n14, B=@n16, Alpha=@n15.RadialGradientExponential), B=LinearInterpolate(A=@n33, B=@n34) {constAlpha=0.75}, Alpha=@n1) {constA=0.75}
   Fraction=ScalarParameter() {group=General, parameterName=Desat}
  )
  n36 = VectorParameter() {group=Rarity Colors, parameterName=Rarity 4, defaultValue=<rgba 0.0006 0.0782 0.0009 1>}
  n37 = Constant()
  n38 = ScalarParameter() {defaultValue=1.125, group=Herobackground, parameterName=HeroBar-Y-Offset}
  n39 = MaterialFunctionCall() {materialFunction=&GetUserInterfaceUV}
  n40 = Add(A=Add(A=Multiply(A=@n39.Normalized UV) {constB=2}) {constB=-0.5}, B=Constant2Vector() {g=-0.425})
  n41 = Rotator(Coordinate=Add(A=@n40, B=AppendVector(A=@n37, B=@n38)) {constB=0.95}, Time=Constant() {r=-0.238}) {speed=6.2832}
  n42 = ScalarParameter() {defaultValue=0.5, group=Herobackground, parameterName=HeroBacking-Max}
  n43 = ScalarParameter() {defaultValue=0.49, group=Herobackground, parameterName=HeroBacking-Min}
  n44 = Multiply(A=MaterialFunctionCall(Alpha=ComponentMask(None=@n41) {r=true}, Min=@n42, Max=@n43) {materialFunction=&SmoothStep}.Result) {constB=0.9}
  n45 = LinearInterpolate(A=@n35, B=Multiply(A=@n36.RGB) {constB=1.25}, Alpha=@n44)
  n46 = ScalarParameter() {group=General, parameterName=bIsLead, sliderMax=1}
  n47 = Multiply(A=@n36.RGB) {constB=0.85}
  n48 = Rotator(Coordinate=@n40, Time=Constant() {r=0.205}) {speed=6.2832}
  n49 = ComponentMask(None=Rotator(Coordinate=@n40, Time=Constant() {r=-0.205}) {speed=6.2832}) {r=true}
  n50 = DotProduct(
   A=MaterialFunctionCall(Alpha=ComponentMask(None=@n48) {r=true}, Min=@n43, Max=@n42) {materialFunction=&SmoothStep}.Result
   B=OneMinus(None=MaterialFunctionCall(Alpha=@n49, Min=@n43, Max=@n42) {materialFunction=&SmoothStep}.Result)
  )
  n51 = LinearInterpolate(A=@n47, B=@n36.RGB, Alpha=@n50)
  n52 = VectorParameter() {group=Rarity Colors, parameterName=Rarity 3, defaultValue=<rgba 0 0.2159 0.0015 1>}
  n53 = Multiply(A=@n52.RGB) {constB=0.75}
  n54 = Rotator(Coordinate=@n40, Time=Constant() {r=0.17}) {speed=6.2832}
  n55 = ComponentMask(None=Rotator(Coordinate=@n40, Time=Constant() {r=-0.17}) {speed=6.2832}) {r=true}
  n56 = DotProduct(
   A=MaterialFunctionCall(Alpha=ComponentMask(None=@n54) {r=true}, Min=@n43, Max=@n42) {materialFunction=&SmoothStep}.Result
   B=OneMinus(None=MaterialFunctionCall(Alpha=@n55, Min=@n43, Max=@n42) {materialFunction=&SmoothStep}.Result)
  )
  n57 = Rotator(Coordinate=@n40, Time=Constant() {r=0.135}) {speed=6.2832}
  n58 = ComponentMask(None=Rotator(Coordinate=@n40, Time=Constant() {r=-0.135}) {speed=6.2832}) {r=true}
  n59 = DotProduct(
   A=MaterialFunctionCall(Alpha=ComponentMask(None=@n57) {r=true}, Min=@n43, Max=@n42) {materialFunction=&SmoothStep}.Result
   B=OneMinus(None=MaterialFunctionCall(Alpha=@n58, Min=@n43, Max=@n42) {materialFunction=&SmoothStep}.Result)
  )
  n60 = Rotator(Coordinate=@n40, Time=Constant() {r=0.1}) {speed=6.2832}
  n61 = ComponentMask(None=Rotator(Coordinate=@n40, Time=Constant() {r=-0.1}) {speed=6.2832}) {r=true}
  n62 = DotProduct(
   A=MaterialFunctionCall(Alpha=ComponentMask(None=@n60) {r=true}, Min=@n43, Max=@n42) {materialFunction=&SmoothStep}.Result
   B=OneMinus(None=MaterialFunctionCall(Alpha=@n61, Min=@n43, Max=@n42) {materialFunction=&SmoothStep}.Result)
  )
  n63 = LinearInterpolate(
   A=LinearInterpolate(A=LinearInterpolate(A=@n51, B=@n53, Alpha=@n56), B=@n52.RGB, Alpha=@n59)
   B=Multiply(A=VectorParameter() {group=Rarity Colors, parameterName=Rarity 2, defaultValue=<rgba 0.1195 0.521 0 1>}.RGB) {constB=0.75}
   Alpha=@n62
  )
  n64 = Constant2Vector() {g=1.6172, r=1.25}
  n65 = Multiply(
   A=MaterialFunctionCall(
    UVs=Multiply(A=@n39.Normalized UV, B=@n64)
    CenterPosition=Constant2Vector() {g=0.25, r=0.6125}
    Radius=ScalarParameter() {defaultValue=0.75, group=Herobackground, parameterName=HeroHighlight-Radius}
    Density=ScalarParameter() {defaultValue=1, group=Herobackground, parameterName=HeroHighlight-Density}
   ) {materialFunction=&RadialGradientExponential}.RadialGradientExponential
   B=VectorParameter() {group=Rarity Colors, parameterName=Rarity 1, defaultValue=<rgba 0.3564 0.9911 0 1>}.RGB
  ) {constB=1.75}
  n66 = TextureCoordinate() {coordinateIndex=4}
  n67 = Add(A=Multiply(A=@n66, B=Constant2Vector() {g=1.6172, r=1}) {constB=1.5}, B=Constant2Vector())
  n68 = LinearInterpolate(
   A=ScalarParameter() {defaultValue=0.25, group=Herobackground, parameterName=PanSpeedFactor-Slow}
   B=ScalarParameter() {defaultValue=1, group=Herobackground, parameterName=PanSpeedFactor-Fast}
   Alpha=ScalarParameter() {group=Herobackground, parameterName=bAnimateFast}
  ) {constA=0.25}
  n69 = Multiply(A=Constant2Vector() {g=0.05}, B=@n68)
  n70 = ScalarParameter() {defaultValue=0.1, group=Herobackground, parameterName=TriangleMin}
  n71 = ScalarParameter() {defaultValue=0.4, group=Herobackground, parameterName=TriangleMax}
  n72 = MaterialFunctionCall(
   Alpha=TextureSample(UVs=Panner(Coordinate=@n67, Speed=@n69) {speedY=0.05}) {samplerType=SAMPLERTYPE_LinearColor, texture=/Game/UI/Foundation/Items/Textures/T-Card-Hero-Anim-Triangle}.R
   Min=@n70
   Max=@n71
  ) {materialFunction=&SmoothStep}
  n73 = Multiply(A=Constant2Vector() {g=0.1}, B=@n68)
  n74 = MaterialFunctionCall(
   Alpha=TextureSample(UVs=Panner(Coordinate=@n67, Speed=@n73) {speedY=0.1}) {samplerType=SAMPLERTYPE_LinearColor, texture=&T-Card-Hero-Anim-Triangle}.G
   Min=@n70
   Max=@n71
  ) {materialFunction=&SmoothStep}
  n75 = LinearInterpolate(A=@n72.Result, Alpha=@n74.Result)
  n76 = LinearInterpolate(A=LinearInterpolate(A=LinearInterpolate(Alpha=@n50) {constA=0.2, constB=0.4}, Alpha=@n56) {constB=0.6}, Alpha=@n59) {constB=0.8}
  n77 = MaterialFunctionCall(
   UVs=Multiply(A=TextureCoordinate() {coordinateIndex=4}, B=Constant2Vector() {g=1.6172, r=1.25})
   CenterPosition=Constant2Vector() {g=1, r=0.6125}
   Radius=ScalarParameter() {defaultValue=1.192, group=Herobackground, parameterName=HeroHighlight-Radius_1}
   Density=LinearInterpolate(
    A=ScalarParameter() {defaultValue=1.45, group=Herobackground, parameterName=Normal Hero Highlight - Density}
    B=ScalarParameter() {defaultValue=2.275, group=Herobackground, parameterName=Special Hero Highlight - Density}
    Alpha=ScalarParameter() {group=Herobackground, parameterName=bAnimateBorder}
   )
  ) {materialFunction=&RadialGradientExponential}
  n78 = Power(Base=LinearInterpolate(B=@n77.RadialGradientExponential, Alpha=LinearInterpolate(A=@n76, Alpha=@n62)))
  n79 = Multiply(A=Constant2Vector() {g=0.15}, B=@n68)
  n80 = MaterialFunctionCall(
   Alpha=TextureSample(UVs=Panner(Coordinate=@n67, Speed=@n79) {speedY=0.15}) {samplerType=SAMPLERTYPE_LinearColor, texture=&T-Card-Hero-Anim-Triangle}.B
   Min=@n70
   Max=@n71
  ) {materialFunction=&SmoothStep}
  n81 = VectorParameter() {group=Rarity Colors, parameterName=Rarity 4, defaultValue=<rgba 0.0006 0.0782 0.0009 1>}
  n82 = Multiply(A=@n81.RGB) {constB=0.5}
  n83 = ComponentMask(None=@n66) {g=true}
  n84 = VectorParameter() {group=Rarity Colors, parameterName=Rarity 1, defaultValue=<rgba 0.3564 0.9911 0 1>}
  n85 = LinearInterpolate(
   A=LinearInterpolate(B=LinearInterpolate(A=@n81.RGB, B=@n82, Alpha=@n83), Alpha=@n74.Result)
   B=LinearInterpolate(A=@n84.RGB, B=Multiply(A=@n84.RGB) {constB=4}, Alpha=@n77.RadialGradientExponential)
   Alpha=@n77.RadialGradientExponential
  )
  n86 = VectorParameter() {group=Rarity Colors, parameterName=Rarity 2, defaultValue=<rgba 0.1195 0.521 0 1>}
  n87 = LinearInterpolate(A=@n86.RGB, B=VectorParameter() {group=Rarity Colors, parameterName=Rarity 3, defaultValue=<rgba 0 0.2159 0.0015 1>}.RGB, Alpha=@n83)
  n88 = Multiply(A=VectorParameter() {group=Rarity Colors, parameterName=Rarity 1, defaultValue=<rgba 0.3564 0.9911 0 1>}.RGB) {constB=0.75}
  n89 = LinearInterpolate(
   A=Multiply(A=Add(A=@n65, B=@n63), B=ScalarParameter() {defaultValue=1, group=Herobackground, parameterName=HeroBack-Brightness}) {constB=0.75}
   B=LinearInterpolate(A=LinearInterpolate(A=@n85, B=@n87, Alpha=@n72.Result), B=LinearInterpolate(A=@n88, B=@n86.RGB, Alpha=@n83), Alpha=@n80.Result)
   Alpha=LinearInterpolate(A=LinearInterpolate(A=@n75, Alpha=@n78), Alpha=@n80.Result)
  )
  n90 = VectorParameter() {group=Schematic, parameterName=sSchematic-Light-BG, defaultValue=<rgba 0.85 0.85 0.85 1>}
  n91 = VectorParameter() {group=Rarity Colors, parameterName=Rarity 2, defaultValue=<rgba 0.1195 0.521 0 1>}
  n92 = Constant() {desc="Rounded Constant for (X/Y)", r=0.62}
  n93 = Constant() {r=1}
  n94 = Add(
   A=Multiply(A=MaterialFunctionCall() {materialFunction=&GetUserInterfaceUV}.Normalized UV, B=AppendVector(A=@n92, B=@n93))
   B=AppendVector(A=Constant() {desc=X Offset required to Centre X tile of 0.62, r=0.19}, B=Constant())
  )
  n95 = Multiply(A=@n94, B=Constant2Vector() {g=1, r=2})
  n96 = Constant2Vector() {g=-0.25}
  n97 = Add(A=ComponentMask(None=Cosine(None=@n95) {period=0.25}) {r=true}, B=Multiply(A=Add(A=@n96, B=@n94), B=Constant2Vector() {g=4, r=1}) {constB=2})
  n98 = Add(A=Multiply(A=@n94, B=Constant2Vector() {g=8, r=1}) {constB=2}, B=ComponentMask(None=Sine(None=@n95) {period=0.4}) {r=true})
  n99 = Clamp(None=Add(A=ComponentMask(None=@n97) {g=true}, B=ComponentMask(None=@n98) {g=true}))
  n100 = MaterialFunctionCall(Alpha=@n99, Min=Min(A=@n99) {constB=0.75}, Max=Max(A=@n99) {constB=0.85}) {materialFunction=&SmoothStep}
  n101 = Sine(None=@n94) {period=2}
  n102 = Multiply(A=ComponentMask(None=@n101) {r=true}, B=ComponentMask(None=@n101) {g=true})
  n103 = Clamp(None=Multiply(A=Clamp(None=@n100.Result) {maxDefault=0.35}, B=@n102))
  n104 = Multiply(A=@n91.RGB) {constB=0.8}
  n105 = LinearInterpolate(
   A=@n90.RGB
   B=VectorParameter() {group=Schematic, parameterName=sSchematic-Dark-BG, defaultValue=<rgba 0.55 0.55 0.55 1>}.RGB
   Alpha=Clamp(None=Multiply(A=@n102) {constB=1.125})
  )
  n106 = Sine(None=@n94) {period=0.075}
  n107 = ComponentMask(None=@n106) {r=true}
  n108 = ScalarParameter() {defaultValue=0.98, desc=Grid Line Clamp Min, group=Schematic, parameterName=sGridMin}
  n109 = ScalarParameter() {defaultValue=1, desc=Grid Line Clamp Max, group=Schematic, parameterName=sGridMax}
  n110 = ComponentMask(None=@n106) {g=true}
  n111 = Add(
   A=MaterialFunctionCall(Alpha=@n107, Min=Min(A=@n107, B=@n108), Max=Max(A=@n107, B=@n109)) {materialFunction=&SmoothStep}.Result
   B=MaterialFunctionCall(Alpha=@n110, Min=Min(A=@n110, B=@n108), Max=Max(A=@n110, B=@n109)) {materialFunction=&SmoothStep}.Result
  )
  n112 = LinearInterpolate(
   A=LinearInterpolate(
    A=LinearInterpolate(A=@n35, B=@n45, Alpha=@n46)
    B=LinearInterpolate(A=@n89, B=@n47, Alpha=@n44)
    Alpha=ScalarParameter() {defaultValue=1, group=General, parameterName=bIsHero, sliderMax=1}
   )
   B=LinearInterpolate(A=LinearInterpolate(A=@n90.RGB, B=@n91.RGB, Alpha=@n103), B=LinearInterpolate(A=@n104, Alpha=@n105) {constB=0.75}, Alpha=Clamp(None=@n111))
   Alpha=@n7
  )
  n113 = TextureSampleParameter2D(UVs=@n10) {
   group=Icon
   parameterName=Icon
   samplerSource=SSM_Clamp_WorldGroupSettings
   texture=&BonusItemBackground
  }
  n114 = TextureSampleParameter2D(UVs=@n10) {
   group=Icon
   parameterName=Icon
   samplerSource=SSM_Clamp_WorldGroupSettings
   texture=&BonusItemBackground
  }
  n115 = Rotator(Coordinate=Panner(Coordinate=@n10) {speedX=-0.05}, Time=Constant() {r=-0.125}) {speed=6.2831}
  n116 = Power(Base=ComponentMask(None=Sine(None=@n115) {period=0.05}) {g=true}) {constExponent=1.5}
  n117 = ScalarParameter() {defaultValue=0.25, desc=Schematic Animation Clamp Min, group=Schematic, parameterName=gMaskGradientMin}
  n118 = ScalarParameter() {defaultValue=0.75, group=Schematic, parameterName=gMaskGradientMax}
  n119 = Clamp(None=Clamp(None=MaterialFunctionCall(Alpha=@n31, Min=@n117, Max=@n118) {materialFunction=&SmoothStep}.Result))
  n120 = LinearInterpolate(
   B=Multiply(A=ScalarParameter() {defaultValue=1, group=Schematic, parameterName=sSchematicImage-Power}, B=Multiply(A=@n116) {constB=0.25})
   Alpha=@n119
  )
  n121 = LinearInterpolate(
   A=LinearInterpolate(A=@n114.RGB, B=Constant3Vector() {constant=<rgba 0 1 1 0>}, Alpha=LinearInterpolate(B=@n120, Alpha=@n114.A))
   B=@n25
   Alpha=Clamp(None=LinearInterpolate(B=@n119, Alpha=@n114.A)) {maxDefault=0.4}
  )
  n122 = ScalarParameter() {defaultValue=1, group=Schematic, parameterName=sSchematicAnimate}
  n123 = MaterialFunctionCall(UVs=@n10, Alpha=@n113.A) {materialFunction=&ClampAlphaToTransparentBorder}
  n124 = Multiply(A=TextureCoordinate(), B=Constant2Vector() {g=1, r=1})
  n125 = Subtract(
   A=Abs(None=Add(A=@n124) {constB=-0.5})
   B=LinearInterpolate(
    B=ScalarParameter() {defaultValue=0.45, group=Border, parameterName=gStrokeThickness, sliderMax=1}
    Alpha=ScalarParameter() {defaultValue=1, group=Border, parameterName=bBorderOn, sliderMax=1}
   ) {constA=1}
  )
  n126 = MaterialFunctionCall(
   Alpha=MaterialFunctionCall(Vector 2=Max(A=@n125) {constB=0}) {materialFunction=/Engine/Functions/Engine_MaterialFunctions02/Utility/VectorLength}.V2 Length
   Min=ScalarParameter() {group=Border, parameterName=gStrokeMin}
   Max=ScalarParameter() {defaultValue=0.001, group=Border, parameterName=gStrokeMax}
  ) {materialFunction=&SmoothStep}
  LinearInterpolate(
   A=LinearInterpolate(
    A=LinearInterpolate(A=@n112, Alpha=@n13) {constB=0}
    B=LinearInterpolate(A=@n113.RGB, B=@n121, Alpha=@n122)
    Alpha=LinearInterpolate(A=@n123.Result, B=@n123.Result, Alpha=@n122)
   ) {constAlpha=0.8}
   B=@n33
   Alpha=Clamp(None=Multiply(A=@n126.Result))
  )
  ScalarParameter() {defaultValue=1, group=General, parameterName=gAlpha, sliderMax=1}
  TextureCoordinate()
  TextureCoordinate()
 ]
}
