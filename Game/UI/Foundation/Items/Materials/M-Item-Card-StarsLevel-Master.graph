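# UI material that draws a star level icon on item cards, with promoted, available and earned states.
# A star texture's alpha masks the shape; state parameters blend its color from near-black to cyan, gold or white.
# Available adds a smaller gold star fill, pulsed by a sine-driven radial gradient.
# Opacity follows the star alpha times an opacity parameter; the earned pulsing glow is multiplied by zero, so it has no effect.
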Material /Game/UI/Foundation/Items/Materials/M-Item-Card-StarsLevel-Master
graph {
 properties {
  bUsedWithStaticMesh=false
  blendMode=BLEND_Translucent
  materialDomain=MD_UI
  pixelDepthOffsetMode=PDOM_Legacy
  usedShadingModels=Default Lit
 }
 nodes [
  n1 = TextureSample() {samplerSource=SSM_Clamp_WorldGroupSettings, texture=/Game/UI/Foundation/Textures/Icons/Stars/T-Stars-Full-M}
  n2 = LinearInterpolate(B=Constant3Vector() {constant=<rgba 0.0052 0.0052 0.0185 1>}, Alpha=@n1.A)
  n3 = LinearInterpolate(B=Constant3Vector() {constant=<rgba 0 0.8235 1 1>}, Alpha=@n1.A) {constAlpha=0}
  n4 = ScalarParameter() {group=States, parameterName=bPromoted}
  n5 = Constant3Vector() {constant=<rgba 0.7231 0.6445 0.063 1>}
  n6 = Time()
  n7 = LinearInterpolate(Alpha=Sine(None=@n6) {period=2}) {constA=1, constB=2}
  n8 = MaterialFunctionCall(
   UVs=Add(A=TextureCoordinate(), B=Constant2Vector() {g=-0.025})
   Texture Scale=ScalarParameter() {defaultValue=0.45, parameterName=gAvailible Fill Scale, sliderMax=1}
  ) {materialFunction=&ScaleUVsByCenter}
  n9 = Clamp(None=LinearInterpolate(
    B=MaterialFunctionCall(Density=@n7) {materialFunction=&RadialGradientExponential}.RadialGradientExponential
    Alpha=TextureSample(UVs=@n8.UVs) {samplerSource=SSM_Clamp_WorldGroupSettings, texture=&T-Stars-Full-M}.A
   ))
  n10 = ScalarParameter() {group=States, parameterName=bEarned}
  LinearInterpolate(
   A=LinearInterpolate(
    A=LinearInterpolate(A=@n2, B=@n3, Alpha=@n4) {constB=0}
    B=LinearInterpolate(A=@n2, B=@n5, Alpha=@n9)
    Alpha=ScalarParameter() {group=States, parameterName=bAvailible}
   )
   Alpha=@n10
  )
  n11 = Sine(None=@n6) {period=4}
  n12 = Multiply(A=MaterialFunctionCall(Radius=Constant() {r=0.5}, Density=LinearInterpolate(Alpha=@n11) {constA=1.25, constB=2.5}) {materialFunction=&RadialGradientExponential}.RadialGradientExponential) {constB=0}
  Multiply(
   A=ScalarParameter() {defaultValue=1, parameterName=gOpacity, sliderMax=1}
   B=LinearInterpolate(A=@n1.A, B=Add(A=@n1.A, B=@n12), Alpha=@n10) {constB=0}
  )
 ]
}
