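# UI material for an item durability meter that draws a vertical bar filled from the bottom up to DurabilityPercentage.
# An inverted V coordinate is compared against the percentage to split filled from empty area.
# The fill blends linearly from red to green as durability rises, and the empty area evaluates to zero.
# When RepairMode is above zero, the fill turns solid green and the empty part solid red.
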Material /Game/UI/Foundation/Items/Parts/DurabilityMeterMaterial
graph {
 properties {
  bUsedWithStaticMesh=false
  blendMode=BLEND_Translucent
  materialDomain=MD_UI
  pixelDepthOffsetMode=PDOM_Legacy
  usedShadingModels=Default Lit
 }
 nodes [
  n1 = TextureCoordinate()
  n2 = Constant()
  n3 = ScalarParameter() {defaultValue=0.5, parameterName=DurabilityPercentage}
  n4 = Add(
   A=Multiply(A=OneMinus(None=@n3), B=Constant3Vector() {constant=<rgba 1 0 0 1>})
   B=Multiply(A=@n3, B=Constant3Vector() {constant=<rgba 0 1 0 1>})
  )
  n5 = If(
   A=ScalarParameter() {defaultValue=-1, parameterName=RepairMode}
   B=Constant()
   "A > B"=Constant3Vector() {constant=<rgba 0 1 0 1>}
   "A == B"=@n4
   "A < B"=@n4
  )
  If(
   A=OneMinus(None=MaterialFunctionCall(Float2=@n1) {materialFunction=&BreakOutFloat2Components}.G)
   B=@n3
   "A > B"=If(
    A=ScalarParameter() {defaultValue=-1, parameterName=RepairMode}
    B=@n2
    "A > B"=Constant3Vector() {constant=<rgba 1 0 0 1>}
    "A == B"=@n2
    "A < B"=@n2
   )
   "A == B"=@n5
   "A < B"=@n5
  )
 ]
}
