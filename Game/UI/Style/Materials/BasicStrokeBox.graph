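# Translucent UI material that draws a box shape with separately tinted fill and stroke, each with its own opacity.
# A Type scalar picks one of eight columns in the BasicStrokeBox atlas, whose upper row supplies fill and lower row stroke.
# Only texture alpha is used as masks.
# Stroke coverage is subtracted from fill before tinting and adding colors, while summed alphas are clamped for combined opacity.
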Material /Game/UI/Style/Materials/BasicStrokeBox
graph {
 properties {
  bUsedWithStaticMesh=false
  blendMode=BLEND_Translucent
  materialDomain=MD_UI
  pixelDepthOffsetMode=PDOM_Legacy
  shadingModel=MSM_Unlit
  usedShadingModels=Unlit
 }
 nodes [
  n1 = Multiply(A=ScalarParameter() {defaultValue=4, parameterName=Type}) {constB=0.125}
  n2 = Constant()
  n3 = TextureCoordinate() {uTiling=0.125, vTiling=0.5}
  n4 = TextureSampleParameter2D(UVs=Add(A=AppendVector(A=@n1, B=@n2), B=@n3)) {parameterName=FillTexture, texture=/Game/UI/Style/BasicStrokeBox}
  n5 = Constant() {r=0.5}
  n6 = TextureSampleParameter2D(UVs=Add(A=AppendVector(A=@n1, B=@n5), B=@n3)) {parameterName=StrokeTexture, texture=&BasicStrokeBox}
  n7 = VectorParameter() {parameterName=StrokeColor, defaultValue=<rgba 1 1 1 1>}
  n8 = Multiply(A=@n6.A, B=@n7.A)
  n9 = Subtract(A=@n4.A, B=@n8)
  n10 = VectorParameter() {parameterName=FillColor, defaultValue=<rgba 0 0 0 1>}
  Add(A=Clamp(None=Multiply(A=@n9, B=@n10.RGB)), B=Clamp(None=Multiply(A=@n6.A, B=@n7.RGB)))
  Clamp(None=Add(A=Multiply(A=@n4.A, B=@n10.A), B=@n8))
 ]
}
