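# Modal lightbox widget on FortQuantitySelector for choosing an item quantity, showing an item card, current/max amounts and optional info.
# A normalized slider and Min/Max buttons set CurrentQuantity between 0 or 1 (bAllowZeroQuantity) and MaxValue, broadcasting each change.
# Confirm broadcasts quantity and item; Cancel uses the same dispatcher without setting a value.
# Input type swaps slider style, using the left-stick glyph as gamepad thumb; mobile changes button styles.
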WidgetBlueprint /Game/UI/Foundation/Items/QuantitySelector
graphs [
 EventGraph {nodes [
   "On Button Clicked (MinButton)"!() -> (OutputDelegate, then, Button:Common Button Legacy Object Reference) {is=title K2Node_ComponentBoundEvent}
   > HandleMinPressed!(execute, self) -> (then, PassThrough:bool)
   "On Button Clicked (Button_Confirm)"!() -> (OutputDelegate, then, Button:Common Button Legacy Object Reference) {is=title K2Node_ComponentBoundEvent}
   > Call On Button Clicked Dispatcher!(
    execute:Exec
    self:Quantity Selector Object Reference
    ChosenValue:int=Get CurrentQuantity() -> (CurrentQuantity:int=0)
    FortItem:Fort Item Object Reference=Get Item() -> (Item:Fort Item Object Reference)
   ) -> (then:Exec) {is=title K2Node_CallDelegate}
   > PlaySound2D(Sound=Get ConfirmButtonSound() -> (ConfirmButtonSound:Sound Base Object Reference))
   "On Button Clicked (MaxButton)"!() -> (OutputDelegate, then, Button:Common Button Legacy Object Reference) {is=title K2Node_ComponentBoundEvent}
   > HandleMaxPressed!(execute, self) -> (then, PassThrough:bool)
   n1 = "On Value Changed (ValueSlider)"!() -> (OutputDelegate, then, Value:float) {is=title K2Node_ComponentBoundEvent}
   > n2 = Branch(Condition=Get bAllowZeroQuantity() -> (bAllowZeroQuantity:bool))
   n3 = Get MaxValue() -> (MaxValue:int=0)
   > n4 = Set CurrentQuantity(CurrentQuantity:int=Round(A=int * float!(A:int=@n3, B:double=@n1.Value) -> (ReturnValue:double)) {is=title K2Node_CallFunction}) -> (Output_Get:int=0)
   SliderValueChanged(execute=[@n4.then, @n6.then])
   n5 = int * float(A=int - int(A=Get MaxValue() -> (MaxValue=0), B=1) {is=title K2Node_CallFunction}, B=@n1.Value)
   n6 = Set CurrentQuantity(execute=@n2.else, CurrentQuantity=int + int(A=Round(A=@n5), B=1) -> (ReturnValue=0) {is=title K2Node_CommutativeAssociativeBinaryOperator}) -> (Output_Get=0)
   "On Button Clicked (Button_Cancel)"!() -> (OutputDelegate, then, Button:Common Button Legacy Object Reference) {is=title K2Node_ComponentBoundEvent}
   > Call On Button Clicked Dispatcher(FortItem=Get Item())
   > PlaySound2D(Sound=Get StorageCancelSound() -> (StorageCancelSound:Sound Base Object Reference))
   n7 = HandleChangeInputMode!() -> (OutputDelegate, then, bNewInputType:ECommonInputType Enum) {is=title K2Node_CustomEvent}
   > SetupForInputMode!(execute, self, IsUsingGamepad:bool="Equal (Enum)"(A=@n7.bNewInputType, B=Gamepad)) -> (then)
   Event Destruct()
   n8 = Get CommonInputSubsystem()
   > Unbind Event from On Input Method Changed(self=@n8, Delegate=@n7)
   Event Construct()
   > Bind Event to On Input Method Changed(self=@n8, Delegate=@n7)
   > n9 = Branch(Condition=Get bShowConfirmButton() -> (bShowConfirmButton:bool))
   n10 = Branch(execute=[@n11.then, @n9.then], Condition=Get bShowCancelButton() -> (bShowCancelButton:bool))
   SetVisibility(execute=@n10.else, self=Get Button_Cancel() -> (Button_Cancel:Common Button Legacy Object Reference), InVisibility=Hidden)
   n11 = SetVisibility(execute=@n9.else, self=Get Button_Confirm() -> (Button_Confirm:Common Button Legacy Object Reference), InVisibility=Hidden)
   n12 = Event OnQuantitySelectorInitialized!() -> (
    OutputDelegate
    then
    InItem:Fort Item Object Reference
    TitleText:"Text (by ref)"
    ConfirmButtonText:"Text (by ref)"
    AdditionalInfoText:"Text (by ref)"
    InitialQuantity:int=0
    MaximumQuantity:int=0
   ) {is=title K2Node_Event}
   > SetItemToRepresent!(
    execute
    self:Fort Item Widget NUI Object Reference=Get ItemCard() -> (ItemCard:Fort Multi Size Item Card Object Reference)
    ItemToRepresent:Fort Item Object Reference=@n12.InItem
   ) -> (then)
   > "SetText (Text)"(self=Get WindowTitle() -> (WindowTitle:Common Text Object Reference), InText=@n12.TitleText)
   > Set Text!(
    execute
    self:Icon Text Button Object Reference=Cast To IconTextButton!(Object=Get Button_Confirm()) -> (AsIcon Text Button:Icon Text Button Object Reference, bSuccess:bool=~) {is=title K2Node_DynamicCast}
    ButtonText:Text=@n12.ConfirmButtonText
   ) -> (then) {is=title K2Node_CallFunction}
   > SetAdditionalText!(execute, self, AdditionalInfoText:Text=@n12.AdditionalInfoText) -> (then)
   > Set MaxValue(MaxValue:int=@n12.MaximumQuantity) -> (Output_Get:int=0)
   > Set MinimumValue(MinimumValue:int=GetMinValue!() -> (MinValue:int=0)) -> (Output_Get:int=0)
   n13 = int - int(A=Get CurrentQuantity() -> (CurrentQuantity=0), B=Get MinimumValue() -> (MinimumValue:int=0))
   n14 = int - int(A=Get MaxValue() -> (MaxValue=0), B=Get MinimumValue() -> (MinimumValue=0))
   > SetValue!(
    execute
    self:Slider Object Reference=Get ValueSlider() -> (ValueSlider:Analog Slider Object Reference)
    InValue:float=float / float(A="To Float (Integer)"(InInt=@n13), B="To Float (Integer)"(InInt=@n14)) {is=title K2Node_CallFunction}
   ) -> (then)
   n15 = int - int(A=Get MaxValue() -> (MaxValue=0), B=Get MinimumValue() -> (MinimumValue=0))
   > SetStepSize!(execute, self:Slider Object Reference=Get ValueSlider(), InValue:float=float / float(A=1, B="To Float (Integer)"(InInt=@n15))) -> (then)
   > SetCurrentValue!(
    execute
    self:Common Numeric Text Block Object Reference=Get MaxAmount() -> (MaxAmount:Common Numeric Text Block Object Reference)
    NewValue:float="To Float (Integer)"(InInt=Get MaxValue() -> (MaxValue=0))
   ) -> (then)
   > SetCurrentValue(
    self=Get CurrentAmount() -> (CurrentAmount:Common Numeric Text Block Object Reference)
    NewValue="To Float (Integer)"(InInt=Get CurrentQuantity() -> (CurrentQuantity=0))
   )
   Event On Activated()
   n16 = Get CommonInputSubsystem!() -> (ReturnValue:Common Input Object Reference) {is=title K2Node_GetBlueprintContextEx}
   > SetupForInputMode(IsUsingGamepad="Equal (Enum)"(A=GetCurrentInputType(self=@n16), B=Gamepad))
   > n17 = Branch(Condition=IsMobileGame())
   n18 = Get MinButton() -> (MinButton:Icon Text Button Object Reference)
   n19 = Get Button_Cancel()
   n20 = Get Button_Confirm()
   n21 = Get MaxButton() -> (MaxButton:Icon Text Button Object Reference)
   > SetStyle!(
    execute
    self:Common Button Base Object Reference=[@n18, @n19, @n20, @n21]
    InStyle:Common Button Style Class Reference=/FortUILibrary/Foundation/Buttons/ButtonStyles/TopBar/ButtonStyle-Outline-M.ButtonStyle-Outline-M_C
   ) -> (then)
   > SetControllerStyle!(
    execute
    self:Icon Text Button Object Reference=[Cast To IconTextButton(Object=@n19) -> (bSuccess=~), Cast To IconTextButton(Object=@n20) -> (bSuccess=~), @n18, @n21]
    ControllerInputStyle:Common Button Style Class Reference=&ButtonStyle-Outline-M_C
   ) -> (then)
   n22 = Get MinButton()
   n23 = Get Button_Cancel()
   n24 = Get Button_Confirm()
   n25 = Get MaxButton()
   SetStyle(
    execute=@n17.else
    self=[@n22, @n23, @n24, @n25]
    InStyle=/FortUILibrary/Foundation/Buttons/ButtonStyles/TopBar/ButtonStyle-Outline.ButtonStyle-Outline_C
   )
   > SetControllerStyle(
    self=[Cast To IconTextButton(Object=@n23) -> (bSuccess=~), Cast To IconTextButton(Object=@n24) -> (bSuccess=~), @n22, @n25]
    ControllerInputStyle=&ButtonStyle-Outline_C
   )
  ]}
 GetCurrentValue {nodes [
   GetCurrentValue() {is=title K2Node_FunctionEntry}
   > Return Node!(execute, CurrentValue:int=Get CurrentQuantity() -> (CurrentQuantity=0))
  ]}
 GetMinValue {nodes [
   GetMinValue() {is=title K2Node_FunctionEntry}
   > Return Node!(execute, MinValue:int=Select(Option 0:int=1, Option 1:int=0, Index=Get bAllowZeroQuantity()) -> (ReturnValue:int))
  ]}
 HandleMaxPressed {nodes [
   HandleMaxPressed() {is=title K2Node_FunctionEntry}
   > SetCurrentValue(self:Self Object Reference, NewValue:int=Get MaxValue() -> (MaxValue=0))
   > Return Node!(execute, PassThrough:bool=false)
  ]}
 HandleMinPressed {nodes [
   HandleMinPressed() {is=title K2Node_FunctionEntry}
   > SetCurrentValue(NewValue=GetMinValue() -> (MinValue=0) {is=title K2Node_CallFunction})
   > Return Node(PassThrough=false)
  ]}
 OnButtonClickedDispatcher {nodes [OnButtonClickedDispatcher!() -> (then, ChosenValue:int, FortItem:Fort Item Object Reference) {is=title K2Node_FunctionEntry}]}
 OnStackValueChangedDispatcher {nodes [OnStackValueChangedDispatcher!() -> (then, NewValue:int, FortItem:Fort Item Object Reference) {is=title K2Node_FunctionEntry}]}
 SetAdditionalText {nodes [
   n1 = SetAdditionalText!() -> (then, AdditionalInfoText:Text) {is=title K2Node_FunctionEntry}
   n2 = Get AdditionalInfo() -> (AdditionalInfo:Common Text Object Reference)
   > SetVisibility(
    self=@n2
    InVisibility=Select(Option 0:ESlateVisibility Enum=SelfHitTestInvisible, Option 1:ESlateVisibility Enum=Collapsed, Index=TextIsEmpty(InText=@n1)) -> (ReturnValue:ESlateVisibility Enum)
   )
   > "SetText (Text)"(self=@n2, InText=@n1)
   > Return Node!(execute)
  ]}
 SetCurrentValue {nodes [
   SetCurrentValue!() -> (then, NewValue:int) {is=title K2Node_FunctionEntry}
   > n1 = Branch(Condition="integer < integer"!(A:int=Get MaxValue() -> (MaxValue=0), B:int=GetMinValue() -> (MinValue=0)) -> (ReturnValue:bool) {is=title K2Node_CallFunction})
   Set CurrentQuantity(execute=[@n1.then, @n2.then], CurrentQuantity=GetMinValue() -> (MinValue=0)) -> (Output_Get=0)
   > SetValue(self=Get ValueSlider(), InValue=0)
   > SliderValueChanged()
   n2 = Branch(
    execute=@n1.else
    Condition="integer <= integer"!(A:int=Get NewValue() -> (NewValue:int=0), B:int=0) -> (ReturnValue:bool) {is=title K2Node_CallFunction}
   )
   n3 = Branch(
    execute=@n2.else
    Condition="integer >= integer"!(A:int=Get NewValue() -> (NewValue=0), B:int=Get MaxValue() -> (MaxValue=0)) -> (ReturnValue:bool) {is=title K2Node_CallFunction}
   )
   > Set CurrentQuantity(CurrentQuantity=Get MaxValue() -> (MaxValue=0)) -> (Output_Get=0)
   > SetValue(self=Get ValueSlider(), InValue=1)
   > SliderValueChanged()
   Set MinValue(execute=@n3.else, MinValue:int=GetMinValue() -> (MinValue=0)) -> (Output_Get:int=0)
   n4 = int - int(A=Get MaxValue() -> (MaxValue=0), B=Get MinValue() -> (MinValue:int=0))
   n5 = int - int(A=Get NewValue() -> (NewValue=0), B=Get MinValue() -> (MinValue=0))
   > SetValue(self=Get ValueSlider(), InValue=float / float(A="To Float (Integer)"(InInt=@n5), B="To Float (Integer)"(InInt=@n4)))
   n6 = int - int(A=Get MaxValue() -> (MaxValue=0), B=Get MinValue() -> (MinValue=0))
   > SetStepSize(self=Get ValueSlider(), InValue=float / float(A=1, B="To Float (Integer)"(InInt=@n6)))
   > Set CurrentQuantity(CurrentQuantity=Get NewValue() -> (NewValue=0)) -> (Output_Get=0)
   > SliderValueChanged()
  ]}
 SetupForInputMode {nodes [
   SetupForInputMode!() -> (then, IsUsingGamepad:bool=~) {is=title K2Node_FunctionEntry}
   n1 = Get CommonInputSubsystem()
   > n2 = Switch on ECommonInputType!(execute, Selection:ECommonInputType Enum=GetCurrentInputType(self=@n1)) -> (MouseAndKeyboard:Exec, Gamepad:Exec, Touch:Exec, Count:Exec) {is=title K2Node_SwitchEnum}
   n3 = Break Slider Style!(SliderStyle:"Slider Style Structure (by ref)"=Get SliderStyle_Generic() -> (SliderStyle_Generic:Slider Style Structure)) -> (
    NormalBarImage:Slate Brush Structure
    HoveredBarImage:Slate Brush Structure
    DisabledBarImage:Slate Brush Structure
    NormalThumbImage:Slate Brush Structure
    HoveredThumbImage:Slate Brush Structure
    DisabledThumbImage:Slate Brush Structure
    BarThickness:float=~
   ) {is=title K2Node_BreakStruct}
   n4 = GetBrushForKey!(
    self:Fort Global UIContext Object Reference=Get FortGlobalUIContext!() -> (ReturnValue:Fort Global UIContext Object Reference) {is=title K2Node_GetBlueprintContextEx}
    Key:Key Structure=Gamepad_LeftX
   ) -> (Brush:Slate Brush Structure, ReturnValue:bool)
   Set WidgetStyle!(
    execute=@n2.Gamepad
    WidgetStyle:Slider Style Structure=Make Slider Style!(
     NormalBarImage:Slate Brush Structure=@n3
     HoveredBarImage:Slate Brush Structure=@n3.HoveredBarImage
     DisabledBarImage:Slate Brush Structure=@n3.DisabledBarImage
     NormalThumbImage:Slate Brush Structure=@n4
     HoveredThumbImage:Slate Brush Structure=@n4
     DisabledThumbImage:Slate Brush Structure=@n3.DisabledThumbImage
     BarThickness:float=@n3.BarThickness
    ) -> (SliderStyle:Slider Style Structure) {is=title K2Node_MakeStruct}
    self:Slider Object Reference=Get ValueSlider()
   ) -> (then, Output_Get:Slider Style Structure)
   > RestoreCenterWidget!(execute, self:Fort Activatable Panel Object Reference, FallbackWidget:Widget Object Reference=Get ValueSlider()) -> (then)
   Set WidgetStyle(execute=@n2.Touch, WidgetStyle=Get SliderStyle_Touch() -> (SliderStyle_Touch:Slider Style Structure), self=Get ValueSlider())
   Set WidgetStyle(
    execute=@n2.MouseAndKeyboard
    WidgetStyle=Get SliderStyle_MouseKeyboard() -> (SliderStyle_MouseKeyboard:Slider Style Structure)
    self=Get ValueSlider()
   )
   Get Current Gamepad Type!(self:Common Input Object Reference=@n1) -> (ReturnValue:ECommonGamepadType Enum=XboxOneController) {is=title K2Node_CallFunction}
  ]}
 SliderValueChanged {nodes [
   SliderValueChanged() {is=title K2Node_FunctionEntry}
   > SetCurrentValue(
    self:Common Numeric Text Block Object Reference=Get CurrentAmount()
    NewValue:float="To Float (Integer)"(InInt=Get CurrentQuantity() -> (CurrentQuantity=0))
   ) {is=title K2Node_CallFunction}
   > Call On Stack Value Changed Dispatcher!(
    execute:Exec
    self:Quantity Selector Object Reference
    NewValue:int=Get CurrentQuantity() -> (CurrentQuantity=0)
    FortItem:Fort Item Object Reference=Get Item()
   ) -> (then:Exec) {is=title K2Node_CallDelegate}
  ]}
]
variables [
 bAllowZeroQuantity {type=boolean}
 bShowCancelButton {type=boolean, default=true}
 bShowConfirmButton {type=boolean, default=true}
 ConfirmButtonSound {type=SoundBase, default=null}
 MaxValue {type=integer}
 MinimumValue {type=integer}
 OnButtonClickedDispatcher {}
 OnStackValueChangedDispatcher {}
 SliderStyle_Generic {
  type=SliderStyle
  default {
   barThickness=8
   disabledBarImage {
    bIsDynamicallyLoaded=false
    drawAs=Image
    imageType=NoImage
    mirroring=NoMirror
    resourceName=None
    resourceObject=/Game/Environments/World/Utility/128gray
    tiling=NoTile
    imageSize=<256 256>
    margin=<ltrb 0.2 0.2 0.2 0.2>
    outlineSettings {
     bUseBrushTransparency=false
     roundingType=HalfHeightRadius
     width=0
     color {colorUseRule=UseColor_Specified, specifiedColor=<rgba 0 0 0 0>}
     cornerRadii=<0 0 0 1>
    }
    tintColor {colorUseRule=UseColor_Specified, specifiedColor=<rgba 1 1 1 1>}
    uVRegion {bIsValid=false, max=<0 0>, min=<0 0>}
   }
   disabledThumbImage {
    bIsDynamicallyLoaded=false
    drawAs=Image
    imageType=FullColor
    mirroring=NoMirror
    resourceName=../../../Engine/Content/Slate/Common/Button.png
    resourceObject=/Game/UI/Style/Materials/BasicStrokeBox
    tiling=NoTile
    imageSize=<20 55>
    margin=<ltrb 0 0 0 0>
    outlineSettings {
     bUseBrushTransparency=false
     roundingType=HalfHeightRadius
     width=0
     color {colorUseRule=UseColor_Specified, specifiedColor=<rgba 0 0 0 0>}
     cornerRadii=<0 0 0 1>
    }
    tintColor {colorUseRule=UseColor_Specified, specifiedColor=<rgba 1 1 1 1>}
    uVRegion {bIsValid=false, max=<0 0>, min=<0 0>}
   }
   hoveredBarImage {
    bIsDynamicallyLoaded=false
    drawAs=Image
    imageType=NoImage
    mirroring=NoMirror
    resourceName=None
    resourceObject=null
    tiling=NoTile
    imageSize=<256 256>
    margin=<ltrb 0 0 0 0>
    outlineSettings {
     bUseBrushTransparency=false
     roundingType=HalfHeightRadius
     width=0
     color {colorUseRule=UseColor_Specified, specifiedColor=<rgba 0 0 0 0>}
     cornerRadii=<0 0 0 1>
    }
    tintColor {colorUseRule=UseColor_Specified, specifiedColor=<rgba 1 1 1 1>}
    uVRegion {bIsValid=false, max=<0 0>, min=<0 0>}
   }
   hoveredThumbImage {
    bIsDynamicallyLoaded=false
    drawAs=Image
    imageType=NoImage
    mirroring=NoMirror
    resourceName=None
    resourceObject=&BasicStrokeBox
    tiling=NoTile
    imageSize=<20 55>
    margin=<ltrb 0 0 0 0>
    outlineSettings {
     bUseBrushTransparency=false
     roundingType=HalfHeightRadius
     width=0
     color {colorUseRule=UseColor_Specified, specifiedColor=<rgba 0 0 0 0>}
     cornerRadii=<0 0 0 1>
    }
    tintColor {colorUseRule=UseColor_Specified, specifiedColor=<rgba 1 1 1 1>}
    uVRegion {bIsValid=false, max=<0 0>, min=<0 0>}
   }
   normalBarImage {
    bIsDynamicallyLoaded=false
    drawAs=Image
    imageType=NoImage
    mirroring=NoMirror
    resourceName=None
    resourceObject=&128gray
    tiling=NoTile
    imageSize=<256 256>
    margin=<ltrb 0.2 0.2 0.2 0.2>
    outlineSettings {
     bUseBrushTransparency=false
     roundingType=HalfHeightRadius
     width=0
     color {colorUseRule=UseColor_Specified, specifiedColor=<rgba 0 0 0 0>}
     cornerRadii=<0 0 0 1>
    }
    tintColor {colorUseRule=UseColor_Specified, specifiedColor=<rgba 1 1 1 1>}
    uVRegion {bIsValid=false, max=<0 0>, min=<0 0>}
   }
   normalThumbImage {
    bIsDynamicallyLoaded=false
    drawAs=Image
    imageType=FullColor
    mirroring=NoMirror
    resourceName=../../../Engine/Content/Slate/Common/Button.png
    resourceObject=&BasicStrokeBox
    tiling=NoTile
    imageSize=<20 55>
    margin=<ltrb 0 0 0 0>
    outlineSettings {
     bUseBrushTransparency=false
     roundingType=HalfHeightRadius
     width=0
     color {colorUseRule=UseColor_Specified, specifiedColor=<rgba 0 0 0 0>}
     cornerRadii=<0 0 0 1>
    }
    tintColor {colorUseRule=UseColor_Specified, specifiedColor=<rgba 1 1 1 1>}
    uVRegion {bIsValid=false, max=<0 0>, min=<0 0>}
   }
  }
 }
 SliderStyle_MouseKeyboard {
  type=SliderStyle
  default {
   barThickness=8
   disabledBarImage {
    bIsDynamicallyLoaded=false
    drawAs=Image
    imageType=NoImage
    mirroring=NoMirror
    resourceName=None
    resourceObject=&128gray
    tiling=NoTile
    imageSize=<256 256>
    margin=<ltrb 0.2 0.2 0.2 0.2>
    outlineSettings {
     bUseBrushTransparency=false
     roundingType=HalfHeightRadius
     width=0
     color {colorUseRule=UseColor_Specified, specifiedColor=<rgba 0 0 0 0>}
     cornerRadii=<0 0 0 1>
    }
    tintColor {colorUseRule=UseColor_Specified, specifiedColor=<rgba 1 1 1 1>}
    uVRegion {bIsValid=false, max=<0 0>, min=<0 0>}
   }
   disabledThumbImage {
    bIsDynamicallyLoaded=false
    drawAs=Image
    imageType=FullColor
    mirroring=NoMirror
    resourceName=../../../Engine/Content/Slate/Common/Button.png
    resourceObject=&BasicStrokeBox
    tiling=NoTile
    imageSize=<20 55>
    margin=<ltrb 0 0 0 0>
    outlineSettings {
     bUseBrushTransparency=false
     roundingType=HalfHeightRadius
     width=0
     color {colorUseRule=UseColor_Specified, specifiedColor=<rgba 0 0 0 0>}
     cornerRadii=<0 0 0 1>
    }
    tintColor {colorUseRule=UseColor_Specified, specifiedColor=<rgba 1 1 1 1>}
    uVRegion {bIsValid=false, max=<0 0>, min=<0 0>}
   }
   hoveredBarImage {
    bIsDynamicallyLoaded=false
    drawAs=Image
    imageType=NoImage
    mirroring=NoMirror
    resourceName=None
    resourceObject=null
    tiling=NoTile
    imageSize=<256 256>
    margin=<ltrb 0 0 0 0>
    outlineSettings {
     bUseBrushTransparency=false
     roundingType=HalfHeightRadius
     width=0
     color {colorUseRule=UseColor_Specified, specifiedColor=<rgba 0 0 0 0>}
     cornerRadii=<0 0 0 1>
    }
    tintColor {colorUseRule=UseColor_Specified, specifiedColor=<rgba 1 1 1 1>}
    uVRegion {bIsValid=false, max=<0 0>, min=<0 0>}
   }
   hoveredThumbImage {
    bIsDynamicallyLoaded=false
    drawAs=Image
    imageType=NoImage
    mirroring=NoMirror
    resourceName=None
    resourceObject=&BasicStrokeBox
    tiling=NoTile
    imageSize=<20 55>
    margin=<ltrb 0 0 0 0>
    outlineSettings {
     bUseBrushTransparency=false
     roundingType=HalfHeightRadius
     width=0
     color {colorUseRule=UseColor_Specified, specifiedColor=<rgba 0 0 0 0>}
     cornerRadii=<0 0 0 1>
    }
    tintColor {colorUseRule=UseColor_Specified, specifiedColor=<rgba 1 1 1 1>}
    uVRegion {bIsValid=false, max=<0 0>, min=<0 0>}
   }
   normalBarImage {
    bIsDynamicallyLoaded=false
    drawAs=Image
    imageType=NoImage
    mirroring=NoMirror
    resourceName=None
    resourceObject=&128gray
    tiling=NoTile
    imageSize=<256 256>
    margin=<ltrb 0.2 0.2 0.2 0.2>
    outlineSettings {
     bUseBrushTransparency=false
     roundingType=HalfHeightRadius
     width=0
     color {colorUseRule=UseColor_Specified, specifiedColor=<rgba 0 0 0 0>}
     cornerRadii=<0 0 0 1>
    }
    tintColor {colorUseRule=UseColor_Specified, specifiedColor=<rgba 1 1 1 1>}
    uVRegion {bIsValid=false, max=<0 0>, min=<0 0>}
   }
   normalThumbImage {
    bIsDynamicallyLoaded=false
    drawAs=Image
    imageType=FullColor
    mirroring=NoMirror
    resourceName=../../../Engine/Content/Slate/Common/Button.png
    resourceObject=&BasicStrokeBox
    tiling=NoTile
    imageSize=<20 55>
    margin=<ltrb 0 0 0 0>
    outlineSettings {
     bUseBrushTransparency=false
     roundingType=HalfHeightRadius
     width=0
     color {colorUseRule=UseColor_Specified, specifiedColor=<rgba 0 0 0 0>}
     cornerRadii=<0 0 0 1>
    }
    tintColor {colorUseRule=UseColor_Specified, specifiedColor=<rgba 1 1 1 1>}
    uVRegion {bIsValid=false, max=<0 0>, min=<0 0>}
   }
  }
 }
 SliderStyle_Touch {
  type=SliderStyle
  default {
   barThickness=8
   disabledBarImage {
    bIsDynamicallyLoaded=false
    drawAs=Image
    imageType=NoImage
    mirroring=NoMirror
    resourceName=None
    resourceObject=&128gray
    tiling=NoTile
    imageSize=<256 256>
    margin=<ltrb 0.2 0.2 0.2 0.2>
    outlineSettings {
     bUseBrushTransparency=false
     roundingType=HalfHeightRadius
     width=0
     color {colorUseRule=UseColor_Specified, specifiedColor=<rgba 0 0 0 0>}
     cornerRadii=<0 0 0 1>
    }
    tintColor {colorUseRule=UseColor_Specified, specifiedColor=<rgba 1 1 1 1>}
    uVRegion {bIsValid=false, max=<0 0>, min=<0 0>}
   }
   disabledThumbImage {
    bIsDynamicallyLoaded=false
    drawAs=Image
    imageType=FullColor
    mirroring=NoMirror
    resourceName=../../../Engine/Content/Slate/Common/Button.png
    resourceObject=&BasicStrokeBox
    tiling=NoTile
    imageSize=<64 55>
    margin=<ltrb 0 0 0 0>
    outlineSettings {
     bUseBrushTransparency=false
     roundingType=HalfHeightRadius
     width=0
     color {colorUseRule=UseColor_Specified, specifiedColor=<rgba 0 0 0 0>}
     cornerRadii=<0 0 0 1>
    }
    tintColor {colorUseRule=UseColor_Specified, specifiedColor=<rgba 1 1 1 1>}
    uVRegion {bIsValid=false, max=<0 0>, min=<0 0>}
   }
   hoveredBarImage {
    bIsDynamicallyLoaded=false
    drawAs=Image
    imageType=NoImage
    mirroring=NoMirror
    resourceName=None
    resourceObject=null
    tiling=NoTile
    imageSize=<32 32>
    margin=<ltrb 0 0 0 0>
    outlineSettings {
     bUseBrushTransparency=false
     roundingType=HalfHeightRadius
     width=0
     color {colorUseRule=UseColor_Specified, specifiedColor=<rgba 0 0 0 0>}
     cornerRadii=<0 0 0 1>
    }
    tintColor {colorUseRule=UseColor_Specified, specifiedColor=<rgba 1 1 1 1>}
    uVRegion {bIsValid=false, max=<0 0>, min=<0 0>}
   }
   hoveredThumbImage {
    bIsDynamicallyLoaded=false
    drawAs=Image
    imageType=NoImage
    mirroring=NoMirror
    resourceName=None
    resourceObject=null
    tiling=NoTile
    imageSize=<32 32>
    margin=<ltrb 0 0 0 0>
    outlineSettings {
     bUseBrushTransparency=false
     roundingType=HalfHeightRadius
     width=0
     color {colorUseRule=UseColor_Specified, specifiedColor=<rgba 0 0 0 0>}
     cornerRadii=<0 0 0 1>
    }
    tintColor {colorUseRule=UseColor_Specified, specifiedColor=<rgba 1 1 1 1>}
    uVRegion {bIsValid=false, max=<0 0>, min=<0 0>}
   }
   normalBarImage {
    bIsDynamicallyLoaded=false
    drawAs=Image
    imageType=NoImage
    mirroring=NoMirror
    resourceName=None
    resourceObject=&128gray
    tiling=NoTile
    imageSize=<256 256>
    margin=<ltrb 0.2 0.2 0.2 0.2>
    outlineSettings {
     bUseBrushTransparency=false
     roundingType=HalfHeightRadius
     width=0
     color {colorUseRule=UseColor_Specified, specifiedColor=<rgba 0 0 0 0>}
     cornerRadii=<0 0 0 1>
    }
    tintColor {colorUseRule=UseColor_Specified, specifiedColor=<rgba 1 1 1 1>}
    uVRegion {bIsValid=false, max=<0 0>, min=<0 0>}
   }
   normalThumbImage {
    bIsDynamicallyLoaded=false
    drawAs=Image
    imageType=FullColor
    mirroring=NoMirror
    resourceName=../../../Engine/Content/Slate/Common/Button.png
    resourceObject=&BasicStrokeBox
    tiling=NoTile
    imageSize=<64 55>
    margin=<ltrb 0 0 0 0>
    outlineSettings {
     bUseBrushTransparency=false
     roundingType=HalfHeightRadius
     width=0
     color {colorUseRule=UseColor_Specified, specifiedColor=<rgba 0 0 0 0>}
     cornerRadii=<0 0 0 1>
    }
    tintColor {colorUseRule=UseColor_Specified, specifiedColor=<rgba 1 1 1 1>}
    uVRegion {bIsValid=false, max=<0 0>, min=<0 0>}
   }
  }
 }
 StorageCancelSound {type=SoundBase, default=/Game/Sounds/Fort_UI_Sounds/FrontEnd/Fort_UI_Storage_Cancel_Cue}
]
widgets {
 info {
  inheritedWidgetCount=2
  namedSlotCount=1
  parentClass=/Script/FortniteUI.FortQuantitySelector
  rootWidgetClass=/Game/UI/Foundation/Lightbox/Lightbox.Lightbox_C
  widgetCount=25
 }
 widgets [
  Lightbox {widgetClassPath=/Game/UI/Foundation/Lightbox/Lightbox, bIsVariable=true, widget=Lightbox, properties {dimAmount=0.8}}
  Overlay_1 {widgetClassPath=&Overlay, namedSlotHost=Lightbox, widget=Overlay_1}
  SizeBox_105 {
   widgetClassPath=&SizeBox
   parent=Overlay_1
   widget=SizeBox_105
   properties {
    bOverride_MinDesiredHeight=true
    bOverride_MinDesiredWidth=true
    heightOverride=200
    minDesiredHeight=200
    minDesiredWidth=600
    navigation=&self:WidgetTree.SizeBox_105.WidgetNavigation_1
    visibility=Visible
    widthOverride=600
   }
   slot {class=OverlaySlot, properties {horizontalAlignment=HAlign_Center, verticalAlignment=VAlign_Center}}
  }
  CommonBorder_1 {
   widgetClassPath=&CommonBorder
   parent=SizeBox_105
   widget=CommonBorder_1
   properties {style=/FortUILibrary/Styles/Border/Border_Navy_VGrad.Border_Navy_VGrad_C, padding=<ltrb 0 0 0 0>}
   slot {class=SizeBoxSlot}
  }
  VerticalBox_0 {
   widgetClassPath=&VerticalBox
   parent=CommonBorder_1
   widget=VerticalBox_0
   slot {class=BorderSlot, properties {padding=<ltrb 0 0 0 0>}}
  }
  CommonBorder_0 {
   widgetClassPath=&CommonBorder
   parent=VerticalBox_0
   widget=CommonBorder_0
   properties {style=/FortUILibrary/Styles/Border/Border-ModalHeader.Border-ModalHeader_C, padding=<ltrb 8 4 4 8>}
   slot {class=VerticalBoxSlot}
  }
  WindowTitle {
   widgetClassPath=&CommonTextBlock
   bIsVariable=true
   parent=CommonBorder_0
   widget=WindowTitle
   properties {
    style=/FortUILibrary/Styles/Typography/Foundation/TextStyle-Base-M-B.TextStyle-Base-M-B_C
    text="{WindowTitle}"
    font {
     bForceMonospaced=false
     bMaterialIsStencil=false
     fontMaterial=null
     fontObject=&NotoSans
     letterSpacing=0
     monospacedWidth=1
     size=24
     skewAmount=0
     typefaceFontName=NotoSans-Bold
     outlineSettings {
      bApplyOutlineToDropShadows=false
      bMiteredCorners=false
      bSeparateFillAlpha=false
      outlineMaterial=null
      outlineSize=0
      outlineColor=<rgba 0 0 0 1>
     }
    }
    shadowOffset=<0 0>
   }
   slot {class=BorderSlot, properties {padding=<ltrb 8 4 4 8>}}
  }
  HorizontalBox_27 {widgetClassPath=&HorizontalBox, parent=VerticalBox_0, widget=HorizontalBox_27, slot {class=VerticalBoxSlot}}
  ItemCard {
   widgetClassPath=/Script/FortniteUI.FortMultiSizeItemCard
   bIsVariable=true
   parent=HorizontalBox_27
   widget=ItemCard
   properties {designTimePreviewItemType=/SaveTheWorld/Items/Workers/Managers/ManagerDoctor_R_T04}
   slot {class=HorizontalBoxSlot, properties {horizontalAlignment=HAlign_Center, verticalAlignment=VAlign_Top, padding=<ltrb 16 16 16 16>}}
  }
  VerticalBox_22 {
   widgetClassPath=&VerticalBox
   parent=HorizontalBox_27
   widget=VerticalBox_22
   slot {class=HorizontalBoxSlot, properties {padding=<ltrb 0 16 16 16>, size {sizeRule=Fill, value=1}}}
  }
  HorizontalBox_76 {
   widgetClassPath=&HorizontalBox
   parent=VerticalBox_22
   widget=HorizontalBox_76
   slot {class=VerticalBoxSlot, properties {padding=<ltrb 0 14 0 0>, size {sizeRule=Fill, value=1}}}
  }
  CurrentAmount {
   widgetClassPath=/Script/CommonUI.CommonNumericTextBlock
   bIsVariable=true
   parent=HorizontalBox_76
   widget=CurrentAmount
   properties {
    currentNumericValue=555
    style=/FortUILibrary/Styles/Typography/Foundation/TextStyle-Base-M-B-70pc.TextStyle-Base-M-B-70pc_C
    text=Text Block
    colorAndOpacity {colorUseRule=UseColor_Specified, specifiedColor=<rgba 1 1 1 0.7>}
    font {
     bForceMonospaced=false
     bMaterialIsStencil=false
     fontMaterial=null
     fontObject=&NotoSans
     letterSpacing=0
     monospacedWidth=1
     size=24
     skewAmount=0
     typefaceFontName=NotoSans-Bold
     outlineSettings {
      bApplyOutlineToDropShadows=false
      bMiteredCorners=false
      bSeparateFillAlpha=false
      outlineMaterial=null
      outlineSize=0
      outlineColor=<rgba 0 0 0 1>
     }
    }
    shadowOffset=<0 0>
   }
   slot {class=HorizontalBoxSlot}
  }
  CommonTextBlock_0 {
   widgetClassPath=&CommonTextBlock
   parent=HorizontalBox_76
   widget=CommonTextBlock_0
   properties {
    style=&TextStyle-Base-M-B_C
    text=/
    font {
     bForceMonospaced=false
     bMaterialIsStencil=false
     fontMaterial=null
     fontObject=&NotoSans
     letterSpacing=0
     monospacedWidth=1
     size=24
     skewAmount=0
     typefaceFontName=NotoSans-Bold
     outlineSettings {
      bApplyOutlineToDropShadows=false
      bMiteredCorners=false
      bSeparateFillAlpha=false
      outlineMaterial=null
      outlineSize=0
      outlineColor=<rgba 0 0 0 1>
     }
    }
    shadowOffset=<0 0>
   }
   slot {class=HorizontalBoxSlot}
  }
  MaxAmount {
   widgetClassPath=&CommonNumericTextBlock
   bIsVariable=true
   parent=HorizontalBox_76
   widget=MaxAmount
   properties {
    currentNumericValue=888
    style=&TextStyle-Base-M-B_C
    text=Text Block
    font {
     bForceMonospaced=false
     bMaterialIsStencil=false
     fontMaterial=null
     fontObject=&NotoSans
     letterSpacing=0
     monospacedWidth=1
     size=24
     skewAmount=0
     typefaceFontName=NotoSans-Bold
     outlineSettings {
      bApplyOutlineToDropShadows=false
      bMiteredCorners=false
      bSeparateFillAlpha=false
      outlineMaterial=null
      outlineSize=0
      outlineColor=<rgba 0 0 0 1>
     }
    }
    shadowOffset=<0 0>
   }
   slot {class=HorizontalBoxSlot}
  }
  AdditionalInfo {
   widgetClassPath=&CommonTextBlock
   bIsVariable=true
   parent=HorizontalBox_76
   widget=AdditionalInfo
   properties {
    autoWrapText=true
    style=/FortUILibrary/Styles/Typography/Foundation/TextStyle-Base-M.TextStyle-Base-M_C
    text="{Additional Info}"
    font {
     bForceMonospaced=false
     bMaterialIsStencil=false
     fontMaterial=null
     fontObject=&NotoSans
     letterSpacing=0
     monospacedWidth=1
     size=24
     skewAmount=0
     typefaceFontName=NotoSans-Regular
     outlineSettings {
      bApplyOutlineToDropShadows=false
      bMiteredCorners=false
      bSeparateFillAlpha=false
      outlineMaterial=null
      outlineSize=0
      outlineColor=<rgba 0 0 0 1>
     }
    }
    shadowOffset=<0 0>
   }
   slot {class=HorizontalBoxSlot, properties {padding=<ltrb 8 0 0 0>, size {sizeRule=Fill, value=1}}}
  }
  SizeBox_87 {
   widgetClassPath=&SizeBox
   parent=VerticalBox_22
   widget=SizeBox_87
   properties {
    bOverride_HeightOverride=true
    bOverride_WidthOverride=true
    heightOverride=40
    navigation=&self:WidgetTree.SizeBox_87.WidgetNavigation_0
    widthOverride=200
   }
   slot {class=VerticalBoxSlot, properties {verticalAlignment=VAlign_Top, padding=<ltrb 0 0 0 20>}}
  }
  ValueSlider {
   widgetClassPath=/Script/CommonUI.AnalogSlider
   bIsVariable=true
   parent=SizeBox_87
   widget=ValueSlider
   properties {
    navigation=&self:WidgetTree.ValueSlider.WidgetNavigation_2
    widgetStyle {
     barThickness=8
     disabledBarImage {
      bIsDynamicallyLoaded=false
      drawAs=Image
      imageType=NoImage
      mirroring=NoMirror
      resourceName=None
      resourceObject=&128gray
      tiling=NoTile
      imageSize=<256 256>
      margin=<ltrb 0.2 0.2 0.2 0.2>
      outlineSettings {
       bUseBrushTransparency=false
       roundingType=HalfHeightRadius
       width=0
       color {colorUseRule=UseColor_Specified, specifiedColor=<rgba 0 0 0 0>}
       cornerRadii=<0 0 0 0>
      }
      tintColor {colorUseRule=UseColor_Specified, specifiedColor=<rgba 1 1 1 1>}
      uVRegion {bIsValid=false, max=<0 0>, min=<0 0>}
     }
     disabledThumbImage {
      bIsDynamicallyLoaded=false
      drawAs=Image
      imageType=FullColor
      mirroring=NoMirror
      resourceName=../../../Engine/Content/Slate/Common/Button.png
      resourceObject=&BasicStrokeBox
      tiling=NoTile
      imageSize=<20 55>
      margin=<ltrb 0 0 0 0>
      outlineSettings {
       bUseBrushTransparency=false
       roundingType=HalfHeightRadius
       width=0
       color {colorUseRule=UseColor_Specified, specifiedColor=<rgba 0 0 0 0>}
       cornerRadii=<0 0 0 0>
      }
      tintColor {colorUseRule=UseColor_Specified, specifiedColor=<rgba 1 1 1 1>}
      uVRegion {bIsValid=false, max=<0 0>, min=<0 0>}
     }
     hoveredBarImage {
      bIsDynamicallyLoaded=false
      drawAs=Image
      imageType=NoImage
      mirroring=NoMirror
      resourceName=None
      resourceObject=null
      tiling=NoTile
      imageSize=<256 256>
      margin=<ltrb 0 0 0 0>
      outlineSettings {
       bUseBrushTransparency=false
       roundingType=HalfHeightRadius
       width=0
       color {colorUseRule=UseColor_Specified, specifiedColor=<rgba 0 0 0 0>}
       cornerRadii=<0 0 0 0>
      }
      tintColor {colorUseRule=UseColor_Specified, specifiedColor=<rgba 1 1 1 1>}
      uVRegion {bIsValid=false, max=<0 0>, min=<0 0>}
     }
     hoveredThumbImage {
      bIsDynamicallyLoaded=false
      drawAs=Image
      imageType=FullColor
      mirroring=NoMirror
      resourceName=../../../Engine/Content/Slate/Common/Button.png
      resourceObject=&BasicStrokeBox
      tiling=NoTile
      imageSize=<20 55>
      margin=<ltrb 0 0 0 0>
      outlineSettings {
       bUseBrushTransparency=false
       roundingType=HalfHeightRadius
       width=0
       color {colorUseRule=UseColor_Specified, specifiedColor=<rgba 0 0 0 0>}
       cornerRadii=<0 0 0 0>
      }
      tintColor {colorUseRule=UseColor_Specified, specifiedColor=<rgba 1 1 1 1>}
      uVRegion {bIsValid=false, max=<0 0>, min=<0 0>}
     }
     normalBarImage {
      bIsDynamicallyLoaded=false
      drawAs=Image
      imageType=NoImage
      mirroring=NoMirror
      resourceName=None
      resourceObject=&128gray
      tiling=NoTile
      imageSize=<256 256>
      margin=<ltrb 0.2 0.2 0.2 0.2>
      outlineSettings {
       bUseBrushTransparency=false
       roundingType=HalfHeightRadius
       width=0
       color {colorUseRule=UseColor_Specified, specifiedColor=<rgba 0 0 0 0>}
       cornerRadii=<0 0 0 0>
      }
      tintColor {colorUseRule=UseColor_Specified, specifiedColor=<rgba 1 1 1 1>}
      uVRegion {bIsValid=false, max=<0 0>, min=<0 0>}
     }
     normalThumbImage {
      bIsDynamicallyLoaded=false
      drawAs=Image
      imageType=FullColor
      mirroring=NoMirror
      resourceName=../../../Engine/Content/Slate/Common/Button.png
      resourceObject=&BasicStrokeBox
      tiling=NoTile
      imageSize=<20 55>
      margin=<ltrb 0 0 0 0>
      outlineSettings {
       bUseBrushTransparency=false
       roundingType=HalfHeightRadius
       width=0
       color {colorUseRule=UseColor_Specified, specifiedColor=<rgba 0 0 0 0>}
       cornerRadii=<0 0 0 0>
      }
      tintColor {colorUseRule=UseColor_Specified, specifiedColor=<rgba 1 1 1 1>}
      uVRegion {bIsValid=false, max=<0 0>, min=<0 0>}
     }
    }
   }
   slot {class=SizeBoxSlot}
  }
  HorizontalBox_2 {
   widgetClassPath=&HorizontalBox
   parent=VerticalBox_22
   widget=HorizontalBox_2
   slot {class=VerticalBoxSlot, properties {size {sizeRule=Fill, value=1}}}
  }
  MinButton {
   widgetClassPath=&IconTextButton
   bIsVariable=true
   parent=HorizontalBox_2
   widget=MinButton
   properties {
    buttonText=Min
    onButtonBaseClicked=(&self:WidgetTree.MinButton.HandleOnButtonClicked)
    onButtonBaseDoubleClicked=(&self:WidgetTree.MinButton.HandleOnButtonDoubleClicked)
    onButtonBaseHovered=(&self:WidgetTree.MinButton.HandleOnButtonHovered)
    onButtonBaseUnhovered=(&self:WidgetTree.MinButton.HandleOnButtonUnhovered)
    onSelectedChangedBase=(&self:WidgetTree.MinButton.HandleOnSelectedChanged)
    style=null
    triggeringInputAction {dataTable=&InputActionDataTable, rowName=MinAmount}
   }
   slot {class=HorizontalBoxSlot}
  }
  Spacer_1 {
   widgetClassPath=&Spacer
   parent=HorizontalBox_2
   widget=Spacer_1
   slot {class=HorizontalBoxSlot, properties {size {sizeRule=Fill, value=1}}}
  }
  MaxButton {
   widgetClassPath=&IconTextButton
   bIsVariable=true
   parent=HorizontalBox_2
   widget=MaxButton
   properties {
    buttonText=Max
    onButtonBaseClicked=(&self:WidgetTree.MaxButton.HandleOnButtonClicked)
    onButtonBaseDoubleClicked=(&self:WidgetTree.MaxButton.HandleOnButtonDoubleClicked)
    onButtonBaseHovered=(&self:WidgetTree.MaxButton.HandleOnButtonHovered)
    onButtonBaseUnhovered=(&self:WidgetTree.MaxButton.HandleOnButtonUnhovered)
    onSelectedChangedBase=(&self:WidgetTree.MaxButton.HandleOnSelectedChanged)
    style=null
    triggeringInputAction {dataTable=&InputActionDataTable, rowName=MaxAmount}
   }
   slot {class=HorizontalBoxSlot}
  }
  HorizontalBox_0 {
   widgetClassPath=&HorizontalBox
   parent=VerticalBox_22
   widget=HorizontalBox_0
   slot {class=VerticalBoxSlot, properties {verticalAlignment=VAlign_Bottom, padding=<ltrb 0 20 0 0>, size {sizeRule=Fill, value=1}}}
  }
  Button_Confirm {
   widgetClassPath=&IconTextButton
   bInherited=true
   bIsVariable=true
   parent=HorizontalBox_0
   widget=Button_Confirm
   properties {
    buttonText=Withdraw
    onButtonBaseClicked=(&self:WidgetTree.Button_Confirm.HandleOnButtonClicked)
    onButtonBaseDoubleClicked=(&self:WidgetTree.Button_Confirm.HandleOnButtonDoubleClicked)
    onButtonBaseHovered=(&self:WidgetTree.Button_Confirm.HandleOnButtonHovered)
    onButtonBaseUnhovered=(&self:WidgetTree.Button_Confirm.HandleOnButtonUnhovered)
    onSelectedChangedBase=(&self:WidgetTree.Button_Confirm.HandleOnSelectedChanged)
    style=null
    triggeringInputAction {dataTable=&InputActionDataTable, rowName=AcceptQuantity}
   }
   slot {class=HorizontalBoxSlot}
  }
  Spacer_0 {widgetClassPath=&Spacer, parent=HorizontalBox_0, widget=Spacer_0, properties {size=<20 1>}, slot {class=HorizontalBoxSlot}}
  Button_Cancel {
   widgetClassPath=&IconTextButton
   bInherited=true
   bIsVariable=true
   parent=HorizontalBox_0
   widget=Button_Cancel
   properties {
    buttonText=Cancel
    onButtonBaseClicked=(&self:WidgetTree.Button_Cancel.HandleOnButtonClicked)
    onButtonBaseDoubleClicked=(&self:WidgetTree.Button_Cancel.HandleOnButtonDoubleClicked)
    onButtonBaseHovered=(&self:WidgetTree.Button_Cancel.HandleOnButtonHovered)
    onButtonBaseUnhovered=(&self:WidgetTree.Button_Cancel.HandleOnButtonUnhovered)
    onSelectedChangedBase=(&self:WidgetTree.Button_Cancel.HandleOnSelectedChanged)
    style=null
    triggeringInputAction {dataTable=&InputActionDataTable, rowName=Cancel}
   }
   slot {class=HorizontalBoxSlot}
  }
 ]
}
